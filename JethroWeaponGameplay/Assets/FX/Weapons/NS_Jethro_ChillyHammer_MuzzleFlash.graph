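# Niagara muzzle flash for the Jethro ChillyHammer weapon, built from burst emitters for cores, lens flare, streaks, sparks, smoke and a brief light.
# System modules supply a random scale and bitwise-derived ADS and suppressor flags.
# These flags and User.AmmoType drive core and flare sizes through per-ammo arrays, shrinking for ADS and zeroing the flare when suppressed.
# Cores use four spawn groups as visibility tags, with time-of-day alpha scaling and a user-selected tint.
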NiagaraSystem /JethroWeaponGameplay/Assets/FX/Weapons/NS_Jethro_ChillyHammer_MuzzleFlash
dependencies {
 usedDataInterfaces [
  &NiagaraDataInterfaceArrayFloat
  &NiagaraDataInterfaceCurve
  &NiagaraDataInterfaceArrayFloat2
  &NiagaraDataInterfaceArrayColor
  &NiagaraDataInterfaceMeshRendererInfo
  &NiagaraDataInterfaceVectorCurve
 ]
 usedDynamicInputs []
 usedModules [
  &self:SystemRandomScale
  &SystemState
  &self:SetSystemBoolsfromBitwise
  &self:SystemSpawnScript.SystemScriptSource.SystemScriptGraph.NiagaraNodeAssignment_0.SetVariables_#1
  &self:SystemSpawnScript.SystemScriptSource.SystemScriptGraph.NiagaraNodeAssignment_3.SetVariables_#2
 ]
}
emitters {
 FN_Weapons_Smoke_BETA001 {
  emitterUpdateScript [
   &EmitterState()
   &SpawnBurst_Instantaneous(Spawn Count=&UniformRangedInt(Minimum=2, Maximum=3), Use Spawn Probability=false, UseLoopCountLimit=false)
  ]
  particleSpawnScript [
   &InitializeParticle(
    Color=<rgba 0.9647 0.9317 0.509 1>
    Lifetime Max=0.5
    Lifetime Min=0.2
    Lifetime Mode=::NewEnumerator1
    Sprite Rotation Angle=&RandomRangeFloat(Minimum=0, Maximum=1)
    Sprite Rotation Mode=::NewEnumerator1
    Sprite Size Mode=::NewEnumerator1
    Uniform Sprite Size Max=60
    Uniform Sprite Size Min=40
    UsePositionOffset=false
    Write Material Random=true
   )
   /SetVariables(Particles.InitSpriteRotRate=&UniformRangedFloat(Minimum=-3, Maximum=3)) {isSetParameters=true}
   &AddVelocityInCone(
    Cone Angle=10
    Module.Use Velocity Falloff On Cone Axis=false
    Velocity Strength=&UniformRangedFloat(Minimum=180, Maximum=600)
   )
  ]
  particleUpdateScript [
   &ParticleState()
   &SpriteRotationRate(Rotation Rate=&Multiply_Float(
     A=$Particles.InitSpriteRotRate
     B=&FloatFromCurve(FloatCurve=DataInterface() {propertyValues {Curve {keys [<at -0.0046 1.0113 RCIM_Cubic>,<at 1.0011 0.09 RCIM_Cubic>]}}})
    ))
   &Drag(Use Linear Drag=true, Use Rotational Drag=false)
   &SolveForcesAndVelocity(Clamp Velocity=false, Limit Acceleration=false)
   &ScaleSpriteSize(Uniform Curve Sprite Scale=DataInterface() {propertyValues {Curve {keys [
        {arriveTangent=0.4478, interpMode=RCIM_Cubic, leaveTangent=0.4478, tangentMode=RCTM_User, time=0.0013, value=0.0367}
        {arriveTangent=0.9127, interpMode=RCIM_Cubic, leaveTangent=0.9127, tangentMode=RCTM_User, time=0.5367, value=0.8831}
        {arriveTangent=3.2289, interpMode=RCIM_Cubic, leaveTangent=3.2289, tangentMode=RCTM_User, time=1, value=1.6042}
       ]}}})
   /Game/Effects/Niagara/Modules/Materials/MaterialSmoke(
    Erode:NiagaraFloat=&FloatFromCurve(FloatCurve=DataInterface() {propertyValues {
       bOptimizeLUT=false
       bUseLUT=false
       Curve {keys [
         {arriveTangent=-0.0652, interpMode=RCIM_Cubic, leaveTangent=-0.0652, tangentMode=RCTM_User, time=0.11, value=0.97}
         {arriveTangent=-5.7797, interpMode=RCIM_Cubic, leaveTangent=-5.7797, tangentMode=RCTM_User, time=0.39, value=0.66}
         {arriveTangent=-0.342, interpMode=RCIM_Cubic, leaveTangent=-0.342, tangentMode=RCTM_User, time=0.9954, value=-0.74}
        ]}
      }})
    Shading:NiagaraFloat=&FloatFromCurve(FloatCurve=DataInterface() {propertyValues {
       bOptimizeLUT=false
       bUseLUT=false
       Curve {keys [
         <at 0.07 0.5 RCIM_Cubic>
         {arriveTangent=1.2205, interpMode=RCIM_Cubic, leaveTangent=1.2205, time=0.17, value=1.02}
         {interpMode=RCIM_Cubic, tangentMode=RCTM_User, time=0.3895, value=0.89}
         <at 0.8 0 RCIM_Cubic>
        ]}
      }})
   )
   &CameraOffset(Camera Offset Amount=20)
   &DynamicMaterialParameters(
    Index 0 Param 1=&Add_Float(
     A=$Particles.RandomFloat
     B=&FloatFromCurve(FloatCurve=DataInterface() {propertyValues {Curve {keys [<at 0 1>,<at 1 0>]}}})
    )
    Index 0 Param 2=&FloatFromCurve(FloatCurve=DataInterface() {propertyValues {Curve {keys [
         {arriveTangent=0.8564, interpMode=RCIM_Cubic, leaveTangent=0.8564, tangentMode=RCTM_User, time=0.0046, value=-0.9753}
         {arriveTangent=4.2876, interpMode=RCIM_Cubic, leaveTangent=5.5268, tangentMode=RCTM_User, time=1, value=1}
        ]}}})
    Index 0 Param 3=&FloatFromCurve(FloatCurve=DataInterface() {propertyValues {Curve {keys [
         {arriveTangent=-22.7009, interpMode=RCIM_Cubic, leaveTangent=-9.0491, tangentMode=RCTM_Break, time=0, value=1}
         {arriveTangent=-0.541, interpMode=RCIM_Cubic, leaveTangent=-28.3175, tangentMode=RCTM_Break, time=0.3531, value=0}
        ]}}})
    Index 0 Param 4=&FloatFromCurve(FloatCurve=DataInterface() {propertyValues {Curve {keys [{arriveTangent=-3.2497, interpMode=RCIM_Cubic, leaveTangent=-3.2497, tangentMode=RCTM_User, time=0, value=1}, <at 1 0 RCIM_Cubic>]}}})
    Module.Param0WriteEnabled=true
    Module.Param1WriteEnabled=true
    Module.Param2WriteEnabled=true
    Module.Param3WriteEnabled=true
   )
   &ScaleColor(
    Scale Alpha=&FloatFromCurve(
     FloatCurve=DataInterface() {propertyValues {Curve {keys [<at -0.003 -0.006>,<at 0.1806 0.5257>,<at 0.4581 0.2965>,<at 1 0>]}}}
     Scale Curve=0.8
    )
    Scale RGB=/Niagara/DynamicInputs/ScaleByCurve/ScaleVectorByCurve(
     InitialValue:Vector3f=<1 1 1>
     ScaleCurve:NiagaraDataInterfaceVectorCurve=DataInterface() {propertyValues {
       XCurve {keys [<at 0 1 RCIM_Cubic>, {arriveTangent=-1.7743, tangentMode=RCTM_User, time=1, value=0}]}
       YCurve {keys [<at 0 1 RCIM_Cubic>, {arriveTangent=-1.7743, tangentMode=RCTM_User, time=1, value=0}]}
       ZCurve {keys [<at 0 1 RCIM_Cubic>, {arriveTangent=-1.7743, tangentMode=RCTM_User, time=1, value=0}]}
      }}
     CurveIndex:NiagaraFloat=$Particles.NormalizedAge
    )
    ScaleA=true
    ScaleRGB=true
   )
  ]
  rendererClasses [&NiagaraSpriteRendererProperties]
  renderers [
   {rendererClass=&NiagaraSpriteRendererProperties}
   {rendererClass=&NiagaraSpriteRendererProperties, rendererIndex=1}
  ]
 }
 FN_Weapons_Sparks_Beta001 {
  emitterUpdateScript [
   &EmitterState()
   &SpawnBurst_Instantaneous(
    Spawn Count=&RandomRangeInt(Minimum=15, Maximum=30)
    Use Spawn Probability=false
    UseLoopCountLimit=false
   )
  ]
  particleSpawnScript [
   &InitializeParticle(
    Color Maximum=<rgba 1 0.9385 0.0882 1>
    Color Minimum=<rgba 0.7552 0.6131 0.1908 1>
    Color Mode=::NewEnumerator2
    Lifetime Max=0.8
    Lifetime Min=0.1
    Lifetime Mode=::NewEnumerator1
    Sprite Size=&UniformRangedVector2D(Minimum=<1 2.5>, Maximum=<2 3.5>)
    Sprite Size Mode=::NewEnumerator4
    UsePositionOffset=false
    Write Material Random=true
   )
   &ShapeLocation(MaskSpawnGroup=false, Override Seed=false, Sphere Radius=10, UseOwnerScale=true)
   &AddVelocityInCone(
    Cone Angle=60
    Module.Use Velocity Falloff On Cone Axis=false
    Velocity Strength=&Lerp_Float(
     A=100
     B=300
     Alpha=&FloatFromCurve(
      FloatCurve=DataInterface() {propertyValues {Curve {keys [
          {interpMode=RCIM_Cubic, tangentMode=RCTM_User, time=0, value=0}
          {arriveTangent=1.2065, interpMode=RCIM_Cubic, leaveTangent=1.2065, tangentMode=RCTM_User, time=0.4, value=0.45}
          {interpMode=RCIM_Cubic, tangentMode=RCTM_User, time=1, value=1}
         ]}}}
      CurveIndex=&RandomRangeFloat(Minimum=0, Maximum=1)
     )
    )
   )
   &ApplyInitialForces()
  ]
  particleUpdateScript [
   &ParticleState()
   &AccelerationForce(Acceleration=&Multiply_VectorByFloat(
     Vector=<0 0 -980>
     Float=&FloatFromCurve(FloatCurve=DataInterface() {propertyValues {Curve {keys [
          {interpMode=RCIM_Cubic, leaveTangentWeight=1.0267, tangentMode=RCTM_User, time=0, value=0}
          {arriveTangent=2.1546, interpMode=RCIM_Cubic, leaveTangent=2.1546, tangentMode=RCTM_User, time=1, value=1}
         ]}}})
    ))
   &Drag(Use Linear Drag=true, Use Rotational Drag=true)
   &CurlNoiseForce(
    Noise Frequency=100
    Noise Strength=&FloatFromCurve(
     FloatCurve=DataInterface() {propertyValues {Curve {keys [
         {interpMode=RCIM_Cubic, tangentMode=RCTM_User, time=0, value=0}
         {interpMode=RCIM_Cubic, tangentMode=RCTM_User, time=0.3, value=1}
         {arriveTangent=-3.2335, interpMode=RCIM_Cubic, leaveTangent=-3.2335, tangentMode=RCTM_User, time=1, value=0.2}
        ]}}}
     Scale Curve=2000
    )
    Pan Noise=false
   )
   &SolveForcesAndVelocity(Clamp Velocity=false, Limit Acceleration=false)
   &ScaleSpriteSizeBySpeed(
    Max Scale Factor=<1 2.5>
    Min Scale Factor=<0.75 1.5>
    Sample Scale Factor By Curve=true
    Scale Factor Curve=DataInterface() {propertyValues {Curve {keys [
        {interpMode=RCIM_Cubic, leaveTangentWeight=0.5699, tangentMode=RCTM_User, time=0, value=0}
        {interpMode=RCIM_Cubic, tangentMode=RCTM_User, time=0.6, value=1}
        {arriveTangentWeight=0.4261, interpMode=RCIM_Cubic, tangentMode=RCTM_User, time=1, value=0}
       ]}}}
    Velocity Threshold:NiagaraFloat=&RandomRangeFloat(Minimum=50, Maximum=200)
   )
   &ScaleSpriteSize(Scale Factor=&Vector2DFromFloat(Value=&FloatFromCurve(FloatCurve=DataInterface() {propertyValues {Curve {keys [
          {interpMode=RCIM_Cubic, tangentMode=RCTM_User, time=0, value=1}
          {arriveTangent=-1.2735, interpMode=RCIM_Cubic, leaveTangent=-1.2735, tangentMode=RCTM_User, time=1, value=0}
         ]}}})))
   &ScaleColor(
    Scale RGB=&VectorFromFloat(Value=&Multiply_Float(
      A=&Lerp_Float(
       A=1
       B=30
       Alpha=&FloatFromCurve(
        FloatCurve=DataInterface() {propertyValues {Curve {keys [
            {interpMode=RCIM_Cubic, tangentMode=RCTM_User, time=0, value=0}
            {arriveTangent=0.8492, interpMode=RCIM_Cubic, leaveTangent=0.8492, tangentMode=RCTM_User, time=0.8, value=0.2}
            {arriveTangent=7.3705, interpMode=RCIM_Cubic, leaveTangent=7.3705, tangentMode=RCTM_User, time=1, value=1}
           ]}}}
        CurveIndex=&RandomRangeFloat(Minimum=0, Maximum=1)
       )
      )
      B=1
     ))
    ScaleA=true
    ScaleRGB=true
   )
  ]
  rendererClasses [&NiagaraSpriteRendererProperties]
  renderers [{rendererClass=&NiagaraSpriteRendererProperties}]
 }
 FlashStreaks {
  emitterUpdateScript [
   &EmitterState()
   &SpawnBurst_Instantaneous(
    Spawn Count=&RandomRangeInt(Minimum=4, Maximum=9)
    Use Spawn Probability=false
    UseLoopCountLimit=false
   )
  ]
  particleSpawnScript [
   /SetVariables(Particles.RandomFloat=&UniformRangedFloat(Minimum=0, Maximum=3)) {isSetParameters=true}
   &InitializeParticle(
    Color=<rgba 1.3 1.2237 0.1096 1>
    Lifetime Max=0.11
    Lifetime Min=0.1
    Lifetime Mode=::NewEnumerator1
    Position Offset=<1.25 0 0>
    Sprite Rotation Angle=&RandomRangeFloat(Minimum=-1, Maximum=1)
    Sprite Rotation Mode=::NewEnumerator1
    Sprite Size Max=<13 24>
    Sprite Size Min=<5 19>
    Sprite Size Mode=::NewEnumerator2
    UsePositionOffset=true
    Write Material Random=true
   )
   &SubUVAnimation(End Frame=3, SubUV Animation Mode:ENiagara_SubUVLookupMode=::NewEnumerator2)
   &SpriteFacingAndAlignment(WriteAlignment=true, WriteFacing=true)
  ]
  particleUpdateScript [
   &ParticleState()
   &ScaleColor(
    Scale Alpha=&FloatFromCurve(
     FloatCurve=DataInterface() {propertyValues {Curve {keys [{interpMode=RCIM_Cubic, tangentMode=RCTM_Break, time=0, value=1}, {interpMode=RCIM_Cubic, tangentMode=RCTM_Break, time=1, value=0}]}}}
     Scale Curve=&RandomRangeFloat(Minimum=1, Maximum=5)
    )
    Scale RGB=&VectorFromFloat(Value=1)
    ScaleA=true
    ScaleRGB=true
   )
   &DynamicMaterialParameters(
    Index 0 Param 2=&FloatFromCurve(FloatCurve=DataInterface() {propertyValues {Curve {keys [{interpMode=RCIM_Cubic, tangentMode=RCTM_Break, time=0, value=0.5}, {interpMode=RCIM_Cubic, tangentMode=RCTM_Break, time=1, value=1}]}}})
    Index 0 Param 3=1
    Index 0 Param 4=&FloatFromCurve(FloatCurve=DataInterface() {propertyValues {Curve {keys [
         {interpMode=RCIM_Cubic, tangentMode=RCTM_Break, time=0.15, value=1}
         {arriveTangentWeight=0.4104, interpMode=RCIM_Cubic, tangentMode=RCTM_Break, time=1, value=0}
        ]}}})
    Module.Index0Float0Write=true
    Module.Index0Float1Write=true
    Module.Index0Float2Write=true
    Module.Index0Float3Write=true
   )
  ]
  rendererClasses [&NiagaraSpriteRendererProperties]
  renderers [{rendererClass=&NiagaraSpriteRendererProperties}]
 }
 Front_LensFlare {
  emitterUpdateScript [
   &EmitterState()
   &SpawnBurst_Instantaneous(Use Spawn Probability=false, UseLoopCountLimit=false)
  ]
  particleSpawnScript [&InitializeParticle(
    Lifetime Max=0.11
    Lifetime Min=0.1
    Lifetime Mode=::NewEnumerator1
    Position Offset=<1.25 0 0>
    Position Offset Coordinate Space=::Local
    Sprite Rotation Mode=::NewEnumerator2
    Sprite Size=&Multiply_Vector2D_ByFloat(
     Float=&self:ADS_Muzzle_Adjusted_Float(
      Base Float Array:NiagaraDataInterfaceArrayFloat=DataInterface() {propertyValues {FloatData [0.7,0.85,1.2,0.85,1]}}
      ADS_Scale:NiagaraFloat=0.5
      MuzzleSupressor_Scale:NiagaraFloat=0
      PerWeaponBP_Scale:NiagaraFloat=$User.MuzzleCoreCapScale
      ADS:NiagaraBool=$System.isADS
      MuzzleSupressor:NiagaraBool=$System.isMuzzleSupressor
      System Random Scale:NiagaraFloat=$System.RandomScale
      AmmoType:NiagaraInt32=$User.AmmoType
     )
     Vector2D=<43 43>
    )
    Sprite Size Mode=::NewEnumerator4
    UsePositionOffset=true
    Write Material Random=true
   )]
  particleUpdateScript [
   &ParticleState()
   &ScaleSpriteSize(Uniform Curve Sprite Scale=DataInterface() {propertyValues {Curve {
       defaultValue=0
       keys [
        {arriveTangent=1.3798, interpMode=RCIM_Cubic, leaveTangent=1.3798, tangentMode=RCTM_User, time=0, value=0}
        {arriveTangent=0.25, interpMode=RCIM_Cubic, leaveTangent=0.25, time=0.7, value=1.5}
        {arriveTangent=-8.7921, interpMode=RCIM_Cubic, leaveTangent=-8.7921, tangentMode=RCTM_User, time=1, value=0.25}
       ]
      }}})
   /SetVariables(
    Particles.FlatCameraOffset=-4
    Particles.Streak_01_Size:Vector2f=&Multiply_Vector2D_ByFloat(
     Float=&self:ADS_Muzzle_Adjusted_Float(
      Base Float Array:NiagaraDataInterfaceArrayFloat=DataInterface() {propertyValues {FloatData [0.6,1,1,0.75,1]}}
      ADS_Scale:NiagaraFloat=0.5
      MuzzleSupressor_Scale:NiagaraFloat=0
      PerWeaponBP_Scale:NiagaraFloat=$User.MuzzleCoreCapScale
      ADS:NiagaraBool=$System.isADS
      MuzzleSupressor:NiagaraBool=$System.isMuzzleSupressor
      System Random Scale:NiagaraFloat=$System.RandomScale
      AmmoType:NiagaraInt32=$User.AmmoType
     )
     Vector2D=<80 5>
    )
   ) {isSetParameters=true}
   &CameraOffset(Camera Offset Amount=3)
   &ScaleColor(Scale Alpha=0.8, ScaleA=true, ScaleRGB=true)
   &DynamicMaterialParameters(
    Index 1 Param 1=8
    Module.Index1Float0Write=true
    Module.Index1Float1Write=true
    Module.Index1Float2Write=true
    Module.Index1Float3Write=true
    Write Parameter Index 0=::NewEnumerator0
    Write Parameter Index 1=::NewEnumerator1
   )
  ]
  rendererClasses [&NiagaraSpriteRendererProperties]
  renderers [
   {rendererClass=&NiagaraSpriteRendererProperties}
   {rendererClass=&NiagaraSpriteRendererProperties, rendererIndex=1}
  ]
 }
 Lighting {
  emitterUpdateScript [
   &EmitterState()
   &SpawnBurst_Instantaneous(Use Spawn Probability=false, UseLoopCountLimit=false)
  ]
  particleSpawnScript [&InitializeParticle(
    Lifetime Max=0.11
    Lifetime Min=0.085
    Lifetime Mode=::NewEnumerator1
    Position Offset=<40 0 0>
    UsePositionOffset=true
    Write Material Random=true
   )]
  particleUpdateScript [
   &ParticleState()
   &ScaleColor(Scale RGB=&MakeVectorFromLinearColorRGB(Color=$System.Color Array), ScaleA=true, ScaleRGB=true)
   &FN_Light_Attributes(
    Color=<rgba 1 0.9695 0.4896 1>
    Create Color Attribute=true
    Create Exponent Attribute=true
    Create Radius Attribute=true
    Exponent=2
    Radius=175
   )
   /SetVariables(Particles.Position=&Lerp_Vector(A=<10 0 0>, B=<20 0 0>, Alpha=$Particles.NormalizedAge)) {isSetParameters=true}
  ]
  rendererClasses [&NiagaraLightRendererProperties]
  renderers [{rendererClass=&NiagaraLightRendererProperties, rendererIndex=0}]
 }
 Muzzle_Cores {
  emitterUpdateScript [
   &EmitterState()
   &SpawnBurst_Instantaneous(Use Spawn Probability=false, UseLoopCountLimit=false)
   &SpawnBurst_Instantaneous(Spawn Group=1, Use Spawn Probability=false, UseLoopCountLimit=false) {module=SpawnBurst_Instantaneous001}
   &SpawnBurst_Instantaneous(Spawn Group=2, Use Spawn Probability=false, UseLoopCountLimit=false) {module=SpawnBurst_Instantaneous002}
   &SpawnBurst_Instantaneous(Spawn Group=3, Use Spawn Probability=false, UseLoopCountLimit=false) {module=SpawnBurst_Instantaneous003}
  ]
  particleSpawnScript [
   /SetVariables(Particles.VisibilityTag=&SpawnGroup()) {isSetParameters=true}
   &InitializeParticle(
    Color=<rgba 1 1 1 0.975>
    Lifetime Max=0.15
    Lifetime Min=0.13
    Lifetime Mode=::NewEnumerator1
    Position Offset=<-1 0 0>
    Sprite Size=&self:ADS_Muzzle_Adjusted_2D(
     ADS_Scale:NiagaraFloat=0.7
     MuzzleSupressor_Scale:NiagaraFloat=1
     PerWeaponBP_Scale:NiagaraFloat=1
     ADS:NiagaraBool=$System.isADS
     MuzzleSupressor:NiagaraBool=$System.isMuzzleSupressor
     System Random Scale:NiagaraFloat=$System.RandomScale
     AmmoType:NiagaraInt32=$User.AmmoType
     Base Vector 2D Array:NiagaraDataInterfaceArrayFloat2=DataInterface() {propertyValues {FloatData [<20 90>, <30 100>, <187.5 300>, <50 100>, <1 300>]}}
    )
    Sprite Size Mode=::NewEnumerator4
    Sprite UV Mode=::NewEnumerator1
    UsePositionOffset=true
    Write Material Random=true
   )
   &SpriteFacingAndAlignment(Alignment Coordinate Space=::Local, Sprite Alignment=<1 0 0>, WriteAlignment=true, WriteFacing=false)
  ]
  particleUpdateScript [
   &ParticleState()
   &ScaleColor(
    Scale Alpha=&FloatFromCurve(
     FloatCurve=DataInterface() {propertyValues {Curve {
        defaultValue=0
        keys [
         {arriveTangent=23.3023, interpMode=RCIM_Cubic, leaveTangent=23.3023, tangentMode=RCTM_User, time=0, value=0}
         {arriveTangent=0.1464, interpMode=RCIM_Cubic, leaveTangent=0.1464, tangentMode=RCTM_User, time=0.12, value=1}
         {arriveTangent=-3.7937, interpMode=RCIM_Cubic, leaveTangent=-3.7937, tangentMode=RCTM_User, time=1, value=0}
        ]
       }}}
     Scale Curve=&FNTimeOfDayScalar(Night Scale=1, Morning Scale=1.2, Day Scale=1.5, Evening Scale=1.2)
    )
    ScaleA=true
    ScaleRGB=true
   )
   &DynamicMaterialParameters(
    Index 0 Param 1=&FloatFromCurve(
     FloatCurve=DataInterface() {propertyValues {Curve {keys [
         {arriveTangent=-0.3026, interpMode=RCIM_Cubic, leaveTangent=-0.3026, leaveTangentWeight=0.3077, tangentMode=RCTM_User, time=0, value=0}
         {arriveTangentWeight=0.8354, interpMode=RCIM_Cubic, tangentMode=RCTM_User, time=1, value=1}
        ]}}}
     Scale Curve=0.8
    )
    Index 0 Param 2=&FloatFromCurve(FloatCurve=DataInterface() {propertyValues {Curve {
        defaultValue=1
        keys [
         {interpMode=RCIM_Cubic, tangentMode=RCTM_User, time=0.7, value=0}
         {arriveTangent=9.5071, interpMode=RCIM_Cubic, leaveTangent=9.5071, tangentMode=RCTM_User, time=1, value=1}
        ]
       }}})
    Index 0 Param 3=&MakeCustomFloatFromBool(Bool=&RandomBool(), True Float=1, False Float=0)
    Index 1 Param 1=0.3
    Index 1 Param 2=2
    Index 2 Param 1=2.25
    Index 2 Param 2=1
    Module.Index0Float0Write=true
    Module.Index0Float1Write=true
    Module.Index0Float2Write=true
    Module.Index0Float3Write=false
    Module.Index1Float0Write=true
    Module.Index1Float1Write=true
    Module.Index1Float2Write=false
    Module.Index1Float3Write=false
    Module.Index2Float0Write=true
    Module.Index2Float1Write=true
    Module.Index2Float2Write=false
    Module.Index2Float3Write=false
    Write Parameter Index 1=::NewEnumerator1
    Write Parameter Index 2=::NewEnumerator1
   )
   &FN_Sprite_Attributes(
    Alignment:Vector3f=<0 1 0>
    Camera Offset=2
    Create Alignment Attribute=true
    Create Camera Offset Attribute=true
    Create Facing Attribute=true
    Create Position Attribute=true
    Create Sprite Rotation Attribute=true
    Create SpriteSize Attribute=true
    Facing:Vector3f=<1 0 0>
    Position:NiagaraPosition=<1.25 0 0>
    SpriteRotation:NiagaraFloat=&RandomRangeFloat(Minimum=-360, Maximum=360)
    SpriteSize=&Multiply_Vector2D_ByFloat(
     Float=&FloatFromCurve(FloatCurve=DataInterface() {propertyValues {Curve {
         defaultValue=0
         keys [
          {arriveTangent=0.339, interpMode=RCIM_Cubic, leaveTangent=0.339, tangentMode=RCTM_User, time=0, value=0}
          {arriveTangent=0.25, interpMode=RCIM_Cubic, leaveTangent=0.25, time=0.7, value=1}
          {arriveTangent=-5.4735, interpMode=RCIM_Cubic, leaveTangent=-5.4735, tangentMode=RCTM_User, time=1, value=0.25}
         ]
        }}})
     Vector2D=$System.Muzzle Core Sprite
    )
    Write to Module Namespace=true
   )
   &FN_Sprite_Attributes(
    Color=&MultiplyLinearColor(
     Linear Color=&SelectLinearColorFromArray(
      Array Sampling Mode=::NewEnumerator1
      Direct Array Index=&MakeIntFromFloat(Float=$User.MuzzleColorHueControl, Conversion Mode=::NewEnumerator2)
      Color Selection Array=DataInterface() {propertyValues {ColorData [<rgba -0.1 -0.75 -0.75 0.25>, <rgba -0.75 -0.75 -0.1 0.25>]}}
      Random Seed=?
     )
     Scale Factor=&MakeVector4FromVectorAndFloat(
      "Vector (RGB)"=<1 1 1>
      "Float (Alpha)"=&FloatFromCurve(
       FloatCurve=DataInterface() {propertyValues {Curve {
          defaultValue=0
          keys [
           {arriveTangent=9.0725, interpMode=RCIM_Cubic, leaveTangent=9.0725, tangentMode=RCTM_User, time=0, value=0}
           {arriveTangent=0.1464, interpMode=RCIM_Cubic, leaveTangent=0.1464, tangentMode=RCTM_User, time=0.1408, value=0.9847}
           {arriveTangent=-0.7388, interpMode=RCIM_Cubic, leaveTangent=-0.7388, tangentMode=RCTM_User, time=1, value=0}
          ]
         }}}
       Scale Curve=&FNTimeOfDayScalar(Night Scale=1, Morning Scale=1, Day Scale=1.5, Evening Scale=1)
      )
     )
    )
    Create Color Attribute=true
    Create SpriteSize Attribute=true
    SpriteSize=&self:ADS_Muzzle_Adjusted_2D(
     ADS_Scale:NiagaraFloat=0.7
     MuzzleSupressor_Scale:NiagaraFloat=1
     PerWeaponBP_Scale:NiagaraFloat=1
     ADS:NiagaraBool=$System.isADS
     MuzzleSupressor:NiagaraBool=$System.isMuzzleSupressor
     System Random Scale:NiagaraFloat=1
     AmmoType:NiagaraInt32=$User.AmmoType
     Base Vector 2D Array:NiagaraDataInterfaceArrayFloat2=DataInterface() {propertyValues {FloatData [<45 100>, <70 125>, <187.5 300>, <90 125>, <187.5 300>]}}
    )
    Write to Module Namespace=true
   ) {module=FN_Sprite_Attributes001}
   &FN_Sprite_Attributes(
    Create SpriteSize Attribute=true
    SpriteSize=&Vector2DFromFloat(Value=&self:ADS_Muzzle_Adjusted_Float(
      Base Float Array:NiagaraDataInterfaceArrayFloat=DataInterface() {propertyValues {FloatData [90,100,125,150,125]}}
      ADS_Scale:NiagaraFloat=0.6
      MuzzleSupressor_Scale:NiagaraFloat=1
      PerWeaponBP_Scale:NiagaraFloat=1
      ADS:NiagaraBool=$System.isADS
      MuzzleSupressor:NiagaraBool=$System.isMuzzleSupressor
      System Random Scale:NiagaraFloat=1
      AmmoType:NiagaraInt32=$User.AmmoType
     ))
    Write to Module Namespace=true
   ) {module=FN_Sprite_Attributes002}
  ]
  rendererClasses [&NiagaraSpriteRendererProperties]
  renderers [
   {rendererClass=&NiagaraSpriteRendererProperties}
   {rendererClass=&NiagaraSpriteRendererProperties, rendererIndex=1}
   {rendererClass=&NiagaraSpriteRendererProperties, rendererIndex=2}
   {rendererClass=&NiagaraSpriteRendererProperties, rendererIndex=3}
  ]
 }
}
summary {
 systemName=NS_Jethro_ChillyHammer_MuzzleFlash
 emitters [
  {emitterName=Lighting, rendererClasses [&NiagaraLightRendererProperties]}
  {emitterName=Front_LensFlare, rendererClasses [&NiagaraSpriteRendererProperties]}
  {emitterName=Muzzle_Cores, rendererClasses [&NiagaraSpriteRendererProperties]}
  {emitterName=FN_Weapons_Sparks_Beta001, rendererClasses [&NiagaraSpriteRendererProperties]}
  {emitterName=FlashStreaks, rendererClasses [&NiagaraSpriteRendererProperties]}
  {emitterName=FN_Weapons_Smoke_BETA001, rendererClasses [&NiagaraSpriteRendererProperties]}
 ]
 userVariables [
  User.AmmoType {type=NiagaraInt32}
  User.MuzzleColorHueControl {type=NiagaraFloat, defaultValue=0, description=""}
  User.MuzzleCoreCapScale {type=NiagaraFloat, defaultValue=0, description=""}
  User.MuzzleModsBitSet {type=NiagaraInt32}
 ]
}
