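# Niagara effect for the Jethro weapon's clear-shell reload: a brief orange flash with rising smoke.
# A burst Flash sprite shrinks and fades, while a curl-noise ribbon trail and rotating, growing smoke sprites drift upward, partly following source movement.
# Loop-age curves taper lifetime and velocity and stop the hot-orange start on later particles.
# User parameters set ribbon lifetime, width and emissive smoke; a time-of-day function scales ribbon opacity.
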NiagaraSystem /JethroWeaponGameplay/Assets/FX/Weapons/NS_Jethro_ClearShell_Reload_Smoke
dependencies {
 usedDataInterfaces [
  &NiagaraDataInterfaceCurve
  &NiagaraDataInterfaceColorCurve
  &NiagaraDataInterfaceVector2DCurve
  &NiagaraDataInterfaceSkeletalMesh
 ]
 usedDynamicInputs []
 usedModules [
  &self:SystemSpawnScript.SystemScriptSource.SystemScriptGraph.NiagaraNodeAssignment_0.SetVariables_#1
  &SystemState
 ]
}
emitters {
 Flash {
  emitterUpdateScript [
   &EmitterState()
   &SpawnBurst_Instantaneous(Use Spawn Probability=false, UseLoopCountLimit=false)
  ]
  particleSpawnScript [
   &InitializeParticle(
    Color=<rgba 1.75 0.1202 0 0.2>
    Lifetime=0.6
    Sprite Size Mode=::NewEnumerator3
    Uniform Sprite Size=12
    UsePositionOffset=false
    Write Material Random=true
   )
   &CameraOffset(Camera Offset Amount=20)
  ]
  particleUpdateScript [
   &ParticleState()
   &ScaleSpriteSize(Uniform Curve Sprite Scale=DataInterface() {propertyValues {Curve {keys [{interpMode=RCIM_Cubic, tangentMode=RCTM_User, time=0, value=2}, {interpMode=RCIM_Cubic, tangentMode=RCTM_User, time=1, value=0}]}}})
   &ScaleColor(
    Scale Alpha=&FloatFromCurve(
     FloatCurve=DataInterface() {propertyValues {Curve {keys [
         {tangentMode=RCTM_User, time=0, value=1}
         {arriveTangent=-0.0405, interpMode=RCIM_Cubic, leaveTangent=-0.0405, tangentMode=RCTM_User, time=0.7009, value=1}
         {arriveTangent=-8.0089, interpMode=RCIM_Cubic, leaveTangent=-8.0089, tangentMode=RCTM_User, time=1, value=0}
        ]}}}
     Scale Curve=0.8
    )
    ScaleA=true
    ScaleRGB=true
   )
  ]
  rendererClasses [&NiagaraSpriteRendererProperties]
  renderers [{rendererClass=&NiagaraSpriteRendererProperties}]
 }
 Reload_Smoke001 {
  emitterUpdateScript [
   &EmitterState()
   &SpawnRate(SpawnRate=60, Use Spawn Probability=false)
  ]
  particleSpawnScript [
   &InitializeParticle(
    Lifetime=&FloatFromCurve(
     FloatCurve=DataInterface() {propertyValues {Curve {keys [<at 0 1 RCIM_Cubic>, {arriveTangent=-1.028, interpMode=RCIM_Cubic, leaveTangent=-1.028, tangentMode=RCTM_User, time=1, value=0.3}]}}}
     CurveIndex=$Emitter.NormalizedLoopAge
     Scale Curve=$User.RibbonLifetime
    )
    Ribbon Width=&Multiply_Float(A=7.5, B=$User.RibbonScale)
    Ribbon Width Mode=::NewEnumerator1
    UsePositionOffset=false
    Write Material Random=true
   )
   &AddVelocity(
    Rotation Coordinate Space=::World
    Velocity=&MakeVector(
     X=0
     Y=0
     Z=&Multiply_Float(
      A=200
      B=&FloatFromCurve(FloatCurve=DataInterface() {propertyValues {Curve {keys [<at 0.5 1>,<at 1 0>]}}}, CurveIndex=$Emitter.NormalizedLoopAge)
     )
    )
   )
  ]
  particleUpdateScript [
   &ParticleState()
   &ScaleColor(
    Scale Alpha=&Multiply_Float(
     A=&FloatFromCurve(
      FloatCurve=DataInterface() {propertyValues {Curve {keys [
          {arriveTangent=28.3592, interpMode=RCIM_Cubic, leaveTangent=28.3592, tangentMode=RCTM_User, time=0, value=0}
          {arriveTangent=-2.6395, interpMode=RCIM_Cubic, leaveTangent=-2.6395, tangentMode=RCTM_User, time=0.1, value=1}
          <at 1 0>
         ]}}}
      Scale Curve=&FNTimeOfDayScalar(Night Scale=1, Morning Scale=0.3, Evening Scale=0.5)
     )
     B=&FloatFromCurve(
      FloatCurve=DataInterface() {propertyValues {Curve {keys [<at 0 1>,<at 0.75 1>,<at 1 0>]}}}
      CurveIndex=$Emitter.NormalizedLoopAge
      Scale Curve=0.2
     )
    )
    Scale RGB=&MakeVectorFromLinearColorRGB(Color=&LerpLinearColors(
      StartColor=&ColorFromCurve(ColorCurve=DataInterface() {propertyValues {
         AlphaCurve {keys [<at 0 1 RCIM_Cubic>]}
         BlueCurve {keys [<at 0.0152 0 RCIM_Cubic>, {arriveTangent=0.1069, time=0.265, value=1}, <at 0.8636 1>, <at 0.9852 0.0885>]}
         GreenCurve {keys [<at 0.0152 0.1599 RCIM_Cubic>, {arriveTangent=-0.1502, time=0.265, value=1}, <at 0.8636 1>, <at 0.9852 0.0885>]}
         RedCurve {keys [<at 0.0152 2 RCIM_Cubic>, {arriveTangent=-4.7204, time=0.265, value=1}, <at 0.8636 1>, <at 0.9852 0.0885>]}
        }})
      EndColor=&ColorFromCurve(ColorCurve=DataInterface() {propertyValues {
         AlphaCurve {keys [<at 0 1 RCIM_Cubic>]}
         BlueCurve {keys [{arriveTangent=0.1069, time=0.265, value=1}, <at 0.8636 1>, <at 0.9852 0.0885>]}
         GreenCurve {keys [{arriveTangent=-0.1502, time=0.265, value=1}, <at 0.8636 1>, <at 0.9852 0.0885>]}
         RedCurve {keys [{arriveTangent=-4.7204, time=0.265, value=1}, <at 0.8636 1>, <at 0.9852 0.0885>]}
        }})
      LerpFactor=&FloatFromCurve(FloatCurve=DataInterface() {propertyValues {Curve {keys [<at 0.5 0>,<at 0.6 1>]}}}, CurveIndex=$Emitter.NormalizedLoopAge)
     ))
    ScaleA=true
    ScaleRGB=true
   )
   &DynamicMaterialParameters(
    Index 0 Param 1=&FloatFromCurve(FloatCurve=DataInterface() {propertyValues {Curve {keys [
         {arriveTangent=-1.1022, interpMode=RCIM_Cubic, leaveTangent=-1.1022, tangentMode=RCTM_User, time=0, value=1}
         {arriveTangent=-0.1649, arriveTangentWeight=1.9926, interpMode=RCIM_Cubic, leaveTangent=-0.1649, tangentMode=RCTM_User, time=1, value=0.125}
        ]}}})
    Index 0 Param 3=-3
    Index 0 Param 4=&MakeCustomFloatFromBool(
     Bool=$User.UseEmissiveSmoke
     True Float=&FloatFromCurve(FloatCurve=DataInterface() {propertyValues {Curve {keys [{arriveTangent=-2.94, interpMode=RCIM_Cubic, time=0, value=1}, <at 0.75 0 RCIM_Cubic>]}}})
     False Float=0
    )
    Index 1 Param 1=&FloatFromCurve(FloatCurve=DataInterface() {propertyValues {Curve {keys [<at 0 1 RCIM_Cubic>, {arriveTangent=45.3807, interpMode=RCIM_Cubic, leaveTangent=45.3807, tangentMode=RCTM_User, time=0.5, value=10}]}}})
    Index 1 Param 2=&FloatFromCurve(FloatCurve=DataInterface() {propertyValues {Curve {keys [{arriveTangent=-2.2405, interpMode=RCIM_Cubic, leaveTangent=-2.2405, tangentMode=RCTM_User, time=0, value=1}, <at 1 0 RCIM_Cubic>]}}})
    Module.Index0Float0Write=true
    Module.Index0Float1Write=false
    Module.Index0Float2Write=true
    Module.Index0Float3Write=true
    Module.Index1Float0Write=true
    Module.Index1Float1Write=true
    Module.Index1Float2Write=true
    Module.Index1Float3Write=true
    Write Parameter Index 1=::NewEnumerator1
   )
   &ScaleRibbonWidth(Ribbon Width Scale=&FloatFromCurve(FloatCurve=DataInterface() {propertyValues {Curve {keys [{tangentMode=RCTM_User, time=0, value=1}, {tangentMode=RCTM_User, time=1, value=0.125}]}}}))
   &CurlNoiseForce(
    Bias Noise=false
    Mask Curl Noise=true
    Noise Frequency=4
    Noise Quality / Cost=::NewEnumerator0
    Noise Strength=625
    Pan Noise=true
    Pan Noise Field=<-10 0 0>
   )
   &InheritSourceMovement(
    Applied Position Scale=&Multiply_VectorByFloat(
     Vector=<1 1 1>
     Float=&FloatFromCurve(FloatCurve=DataInterface() {propertyValues {Curve {keys [<at 0 1>,<at 1 0.6>]}}})
    )
    ApplyForce=false
    ApplyPosition=true
    ApplyPositionFalloff=false
   )
   &SolveForcesAndVelocity(Clamp Velocity=false, Limit Acceleration=false)
  ]
  rendererClasses [&NiagaraRibbonRendererProperties]
  renderers [{rendererClass=&NiagaraRibbonRendererProperties}]
 }
 Smoke {
  emitterUpdateScript [
   &EmitterState()
   &SpawnRate(SpawnRate=12, Use Spawn Probability=false)
  ]
  particleSpawnScript [
   &InitializeParticle(
    Color=&RandomRangeLinearColor(Minimum=<rgba 1 1 1 1>, Maximum=<rgba 1 0.8076 0.7917 1>)
    Lifetime=&FloatFromCurve(
     FloatCurve=DataInterface() {propertyValues {Curve {keys [
         {arriveTangent=-0.41, interpMode=RCIM_Cubic, time=0, value=1.4}
         {arriveTangent=-0.9156, interpMode=RCIM_Cubic, leaveTangent=-0.9156, tangentMode=RCTM_User, time=1, value=0.5}
        ]}}}
     CurveIndex=$Emitter.NormalizedLoopAge
    )
    Sprite Rotation Mode=::NewEnumerator2
    Sprite Size Mode=::NewEnumerator1
    Sprite UV Mode=::NewEnumerator3
    Uniform Sprite Size Max=24
    Uniform Sprite Size Min=16
    UsePositionOffset=false
    Write Material Random=true
   )
   &CameraOffset(Camera Offset Amount=-5)
   &AddVelocity(
    Rotation Coordinate Space=::World
    Velocity=&MakeVector(
     X=0
     Y=0
     Z=&Multiply_Float(
      A=40
      B=&FloatFromCurve(
       FloatCurve=DataInterface() {propertyValues {Curve {keys [<at 0.5 1 RCIM_Cubic>, {arriveTangent=-1.7576, tangentMode=RCTM_User, time=1, value=0.5}]}}}
       CurveIndex=$Emitter.NormalizedLoopAge
      )
     )
    )
   )
  ]
  particleUpdateScript [
   &ParticleState()
   &Drag(Drag=0.25, Use Linear Drag=true, Use Rotational Drag=true)
   &InheritSourceMovement(
    Applied Position Scale=&VectorFromFloat(Value=&FloatFromCurve(FloatCurve=DataInterface() {propertyValues {Curve {keys [<at 0 1>,<at 1 0.8>]}}}))
    ApplyForce=false
    ApplyPosition=true
    ApplyPositionFalloff=false
   )
   &SolveForcesAndVelocity(Clamp Velocity=false, Limit Acceleration=false)
   &DynamicMaterialParameters(
    Index 0 Param 1=&MakeCustomFloatFromBool(Bool=&RandomBool(), True Float=0, False Float=0.75)
    Module.Param0WriteEnabled=true
    Module.Param1WriteEnabled=true
    Module.Param2WriteEnabled=true
    Module.Param3WriteEnabled=true
   )
   &ScaleColor(
    Scale Alpha=&FloatFromCurve(
     FloatCurve=DataInterface() {propertyValues {Curve {keys [
         {arriveTangent=32.0902, interpMode=RCIM_Cubic, leaveTangent=32.0902, tangentMode=RCTM_User, time=0, value=0}
         {arriveTangent=-1.3718, interpMode=RCIM_Cubic, leaveTangent=-1.3718, tangentMode=RCTM_User, time=0.1, value=1}
         {arriveTangent=-0.0983, tangentMode=RCTM_User, time=1, value=0}
        ]}}}
     Scale Curve=0.7
    )
    Scale RGB=&MakeVectorFromLinearColorRGB(Color=&LerpLinearColors(
      StartColor=&ColorFromCurve(ColorCurve=DataInterface() {propertyValues {
         AlphaCurve {keys [<at 0 1 RCIM_Cubic>]}
         BlueCurve {keys [<at 0.1106 0 RCIM_Cubic>, {arriveTangent=0.1069, time=0.368, value=1}, <at 0.8636 1>]}
         GreenCurve {keys [<at 0.1106 0.1599 RCIM_Cubic>, {arriveTangent=-0.1502, time=0.368, value=1}, <at 0.8636 1>]}
         RedCurve {keys [<at 0.1106 2 RCIM_Cubic>, {arriveTangent=-4.7204, time=0.368, value=1}, <at 0.8636 1>]}
        }})
      EndColor=&ColorFromCurve(ColorCurve=DataInterface() {propertyValues {
         AlphaCurve {keys [<at 0 1 RCIM_Cubic>]}
         BlueCurve {keys [{arriveTangent=0.1069, time=0.265, value=1}, <at 0.8636 1>]}
         GreenCurve {keys [{arriveTangent=-0.1502, time=0.265, value=1}, <at 0.8636 1>]}
         RedCurve {keys [{arriveTangent=-4.7204, time=0.265, value=1}, <at 0.8636 1>]}
        }})
      LerpFactor=&FloatFromCurve(FloatCurve=DataInterface() {propertyValues {Curve {keys [<at 0.4 0>,<at 0.45 1>]}}}, CurveIndex=$Emitter.NormalizedLoopAge)
     ))
    ScaleA=true
    ScaleRGB=true
   )
   &SpriteRotationRate(Rotation Rate=&RandomRangeFloat(Minimum=-70, Maximum=70))
   &ScaleSpriteSize(Uniform Curve Sprite Scale=DataInterface() {propertyValues {Curve {keys [
        {arriveTangent=0.6858, interpMode=RCIM_Cubic, time=0, value=0.35}
        {arriveTangent=3.7956, interpMode=RCIM_Cubic, leaveTangent=3.7956, tangentMode=RCTM_User, time=1, value=2}
       ]}}})
  ]
  rendererClasses [&NiagaraSpriteRendererProperties]
  renderers [{rendererClass=&NiagaraSpriteRendererProperties}]
 }
}
summary {
 systemName=NS_Jethro_ClearShell_Reload_Smoke
 emitters [
  {emitterName=Reload_Smoke001, rendererClasses [&NiagaraRibbonRendererProperties]}
  {emitterName=Flash, rendererClasses [&NiagaraSpriteRendererProperties]}
  {emitterName=Smoke, rendererClasses [&NiagaraSpriteRendererProperties]}
 ]
 userVariables [
  User.BaseColorAlpha {type=LinearColor, description="", defaultValue {struct=LinearColor, value=<rgba 1 1 1 0.3>}}
  User.bIsRocketLauncher {type=NiagaraBool}
  User.RibbonLifetime {type=NiagaraFloat, defaultValue=0.5, description=""}
  User.RibbonMaterial {
   type=MaterialInterface
   description=""
   defaultValue {
    struct=NiagaraExt_VariableValue_Object
    value {object=/Game/Effects/Niagara/MasterMaterials/Reload/M_Reload_SmokeTrail, objectClass=/Script/Engine.MaterialInterface}
   }
  }
  User.RibbonScale {type=NiagaraFloat, defaultValue=1, description=""}
  User.UseEmissiveSmoke {type=NiagaraBool}
 ]
}
